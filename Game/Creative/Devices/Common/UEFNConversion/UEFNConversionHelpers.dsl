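; Blueprint helper library for UEFN conversion of Creative devices, covering option migration, device naming and visibility.
; Option migrators copy non-default old values (boolean or float) into new booleans, reset the old to default, and report migration.
; The channel migrator moves a channel shared by two receivers to an unset target, clearing only the first source.
; Other helpers resolve display labels, read option overrides as integers, and choose editor visibility or hidden-in-game.
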(fn GetDeviceName
  (params (Actor :type Actor%20Object%20Reference))
  (bind _asfort_actor_label_override_interface (Utilities|Casting|CastToFortActorLabelOverrideInterface Actor)
    (:then
      (return (LabelOverride|GetDisplayLabel _asfort_actor_label_override_interface)))
    (:CastFailed
      (return (Utilities|Text|ToText(String) (Utilities|GetDisplayName Actor))))))

(fn GetOptionValueAsInt
  (params (ToyOptionsComp :type Fort%20Actor%20Options%20Component%20Object%20Reference) (Key :type String))
  (bind (_currentoptions _returnvalue) (OptionData|GetPropertyOverrides ToyOptionsComp))
  (bind (_value _returnvalue_2) (Utilities|Map|Find _currentoptions Key))
  ;[0 1 0.7338 0.1] Given a key and a toy options component, will return that value as an integer.
  (Utilities|IsValid ToyOptionsComp
    (:Is%20Valid
      (if _returnvalue
        (if _returnvalue_2
          (return true (Utilities|String|StringToInteger _value))
          (else
            _))
        (else
          _))
      (return false 0))
    (:Is%20Not%20Valid
      _)))

(fn MigrateOptionType_BooleanToBoolean
  (params (OldOptionVariable :type Boolean%20%28by%20ref%29) (OldOptionDefault :type Boolean) (NewOptionVariable :type Boolean%20%28by%20ref%29))
  (Utilities|FlowControl|Sequence
    (:then_0
      (if (!= OldOptionVariable OldOptionDefault)
        (Variables|SetBy-RefVar NewOptionVariable OldOptionVariable)
        (Variables|SetBy-RefVar OldOptionVariable OldOptionDefault)
        (LocalVariables|SetbWasMigrated true)))
    (:then_1
      (return (LocalVariables|GetbWasMigrated)))))

(fn MigrateOptionType_FloatToBoolean
  (params (OldOptionVariable :type Float%20%28double-precision%29%20%28by%20ref%29) (OldOptionDefault :type Float%20%28double-precision%29) (NewOptionVariable :type Boolean%20%28by%20ref%29))
  (Utilities|FlowControl|Sequence
    (:then_0
      (if (!= OldOptionVariable OldOptionDefault)
        (Variables|SetBy-RefVar NewOptionVariable (!= OldOptionVariable 0))
        (Variables|SetBy-RefVar OldOptionVariable OldOptionDefault)
        (LocalVariables|SetbWasMigrated true)))
    (:then_1
      (return (LocalVariables|GetbWasMigrated)))))

(fn MigrateToReceiverWhenSameChannel
  (params (SourceComponentA :type Fort%20Gameplay%20Receiver%20Message%20Component%20Object%20Reference) (SourceComponentB :type Fort%20Gameplay%20Receiver%20Message%20Component%20Object%20Reference) (TargetComponent :type Fort%20Gameplay%20Receiver%20Message%20Component%20Object%20Reference))
  (Utilities|FlowControl|Sequence
    (:then_0
      (bind _sourcecomponenta (Variables|GetSourceComponentA)
        (:then
          (bind _sourcecomponentb (Variables|GetSourceComponentB)
            (:then
              (bind _targetcomponent (Variables|GetTargetComponent)
                (:then
                  (if (Math|Boolean|ANDBoolean (Class|FortGameplayMessageComponentBase|IsChannelIdSet _sourcecomponenta) (GameplayTags|Equal(GameplayTagContainer) (Class|FortGameplayMessageComponentBase|GetChannelId _sourcecomponenta) (Class|FortGameplayMessageComponentBase|GetChannelId _sourcecomponentb)) (not (Class|FortGameplayMessageComponentBase|IsChannelIdSet _targetcomponent)))
                    (Class|FortGameplayMessageComponentBase|SetChannelId (Variables|GetTargetComponent) (Class|FortGameplayMessageComponentBase|GetChannelId _sourcecomponenta))
                    (Class|FortGameplayMessageComponentBase|SetChannelId (Variables|GetSourceComponentA) (GameplayTags|MakeLiteralGameplayTagContainer))
                    (LocalVariables|SetbMigrated true)))
                (:else)))
            (:else)))
        (:else)))
    (:then_1
      (return (LocalVariables|GetbMigrated)))))

(fn SetComponentHiddenEditor
  (params (Component :type Scene%20Component%20Object%20Reference) (Hidden :type Boolean) (bPropagateToChildren :type Boolean))
  ;[0.8827 0.1953 1 0.25] Sets the component hidden in game if we're in game, or sets visibility if in editor. This is for use cases in Creative/UEFN where visibility is used for generic visibility in game, while we also want to use hidden in game.
  (if (Editor|InEditor)
    (Rendering|SetVisibility Component (not Hidden) bPropagateToChildren)
    (else
      (Development|SetHiddeninGame Component Hidden bPropagateToChildren))))
